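# Physics asset for the M_MED_Indigo backpack cosmetic, giving its dynamic toe and wire bones capsule bodies and joint limits.
# Three front toes and one back toe each hinge to their own MSTR anchor body, swinging on one axis only.
# The top wires form a narrow-swing two-link chain from their anchor.
# The bottom wire root rotates freely on its anchor and fans into four twist-free branches with wider swing, one having a second link.
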PhysicsAsset /BRCosmetics/Accessories/FORT_Backpacks/Backpack_M_MED_Indigo/Meshes/M_MED_Indigo_Pack_Physics
bodies {
 dyn_bot_wireGrp0 [{
   shapeType=Capsule
   length=2.5959
   radius=0.9962
   shapeName=dyn_bot_wireGrp0_capsule
   center=<0.5247 -1.2053 0.7466>
   rotation=<rot 0 -157.1124 -68.2482>
  }]
 dyn_bot_wireGrp0_0 [{
   shapeType=Capsule
   length=3.7346
   radius=1.1822
   shapeName=dyn_bot_wireGrp0_0_capsule
   center=<3.1906 0.1636 0.1863>
   rotation=<rot 0 88.7682 -86.413>
  }]
 dyn_bot_wireGrp0_1 [{shapeType=Capsule, length=2, radius=0.3729, shapeName=dyn_bot_wireGrp0_1_capsule, center=<2.3895 0 0>, rotation=<rot 90 0 0>}]
 dyn_bot_wireGrp0_1_2 [{
   shapeType=Capsule
   length=2
   radius=0.4592
   shapeName=dyn_bot_wireGrp0_1_2_capsule
   center=<3.4407 0.0272 1.1514>
   rotation=<rot 0 88.3497 -68.9636>
  }]
 dyn_bot_wireGrp0_2 [{
   shapeType=Capsule
   length=2.2586
   radius=0.5425
   shapeName=dyn_bot_wireGrp0_2_capsule
   center=<2.23 -0.2037 -0.0173>
   rotation=<rot 0 -96.2576 -83.3222>
  }]
 dyn_bot_wireGrp0_3 [{
   shapeType=Capsule
   length=2.2469
   radius=0.4781
   shapeName=dyn_bot_wireGrp0_3_capsule
   center=<2.2282 0.0054 0.3232>
   rotation=<rot 0 86.5098 -79.5455>
  }]
 dyn_bot_wireGrp_MSTR [{shapeType=Capsule, length=0.505, radius=0.505, shapeName=dyn_bot_wireGrp_MSTR_capsule, rotation=<rot 0 -90 0>}]
 dyn_c_ba_toe [{
   shapeType=Capsule
   length=6.1992
   radius=2.9348
   shapeName=dyn_c_ba_toe_capsule
   center=<5.852 -0.0074 0.6198>
   rotation=<rot 0 -90.0768 -82.6978>
  }]
 dyn_c_ba_toe_MSTR [{shapeType=Capsule, length=0.505, radius=0.505, shapeName=dyn_c_ba_toe_MSTR_capsule, rotation=<rot 0 -90 0>}]
 dyn_c_fr_toe [{
   shapeType=Capsule
   length=7.5662
   radius=5.6757
   shapeName=dyn_c_fr_toe_capsule
   center=<6.7219 0.0754 0.0129>
   rotation=<rot 0 -89.47 -85.7682>
  }]
 dyn_c_fr_toe_MSTR [{shapeType=Capsule, length=0.505, radius=0.505, shapeName=dyn_c_fr_toe_MSTR_capsule, rotation=<rot 0 -90 0>}]
 dyn_l_fr_toe [{
   shapeType=Capsule
   length=6.422
   radius=5.7501
   shapeName=dyn_l_fr_toe_capsule
   center=<5.9352 -0.181 -0.8484>
   rotation=<rot 0 -93.3797 -57.4513>
  }]
 dyn_l_fr_toe_MSTR [{shapeType=Capsule, length=0.505, radius=0.505, shapeName=dyn_l_fr_toe_MSTR_capsule, rotation=<rot 0 -90 0>}]
 dyn_r_fr_toe [{
   shapeType=Capsule
   length=6.4346
   radius=5.7692
   shapeName=dyn_r_fr_toe_capsule
   center=<5.9045 0.3243 -0.9196>
   rotation=<rot 0 -85.2039 -56.7611>
  }]
 dyn_r_fr_toe_MSTR [{shapeType=Capsule, length=0.505, radius=0.505, shapeName=dyn_l_fr_toe_MSTR_capsule, rotation=<rot 0 -90 0>}]
 dyn_top_wireGrp0_0 [{
   shapeType=Capsule
   length=4.6158
   radius=1.774
   shapeName=dyn_top_wireGrp0_0_capsule
   center=<3.885 0.2666 0.5703>
   rotation=<rot 0 94.2291 -76.3724>
  }]
 dyn_top_wireGrp0_1 [{
   shapeType=Capsule
   length=2.6299
   radius=0.7668
   shapeName=dyn_top_wireGrp0_1_capsule
   center=<2.5367 -0.12 0.5194>
   rotation=<rot 0 85.4929 -79.3854>
  }]
 dyn_top_wireGrp_MSTR [{shapeType=Capsule, length=0.505, radius=0.505, shapeName=dyn_top_wireGrp_MSTR_capsule, rotation=<rot 0 -90 0>}]
}
constraints [
 {bone1Name=dyn_l_fr_toe, bone2Name=dyn_l_fr_toe_MSTR, swing1LimitDegrees=0, swing2LimitDegrees=25, twistLimitDegrees=0}
 {bone1Name=dyn_r_fr_toe, bone2Name=dyn_r_fr_toe_MSTR, swing1LimitDegrees=0, swing2LimitDegrees=25, twistLimitDegrees=0}
 {bone1Name=dyn_c_fr_toe, bone2Name=dyn_c_fr_toe_MSTR, swing1LimitDegrees=0, swing2LimitDegrees=25, twistLimitDegrees=0}
 {bone1Name=dyn_c_ba_toe, bone2Name=dyn_c_ba_toe_MSTR, swing1LimitDegrees=0, swing2LimitDegrees=25, twistLimitDegrees=0}
 {bone1Name=dyn_top_wireGrp0_0, bone2Name=dyn_top_wireGrp_MSTR, swing1LimitDegrees=0, swing2LimitDegrees=15, twistLimitDegrees=0}
 {bone1Name=dyn_top_wireGrp0_1, bone2Name=dyn_top_wireGrp0_0, swing1LimitDegrees=0, swing2LimitDegrees=15, twistLimitDegrees=0}
 {bone1Name=dyn_bot_wireGrp0_2, bone2Name=dyn_bot_wireGrp0, swing1LimitDegrees=30, swing2LimitDegrees=75, twistLimitDegrees=0}
 {bone1Name=dyn_bot_wireGrp0_3, bone2Name=dyn_bot_wireGrp0, swing1LimitDegrees=10, swing2LimitDegrees=60, twistLimitDegrees=0}
 {bone1Name=dyn_bot_wireGrp0_0, bone2Name=dyn_bot_wireGrp0, swing1LimitDegrees=10, swing2LimitDegrees=30, twistLimitDegrees=0}
 {
  bone1Name=dyn_bot_wireGrp0
  bone2Name=dyn_bot_wireGrp_MSTR
  swing1LimitDegrees=45
  swing1Motion=Free
  swing2LimitDegrees=45
  swing2Motion=Free
  twistLimitDegrees=45
  twistMotion=Free
 }
 {bone1Name=dyn_bot_wireGrp0_1_2, bone2Name=dyn_bot_wireGrp0_1, swing1LimitDegrees=20, swing2LimitDegrees=65, twistLimitDegrees=0}
 {bone1Name=dyn_bot_wireGrp0_1, bone2Name=dyn_bot_wireGrp0, swing1LimitDegrees=20, swing2LimitDegrees=70, twistLimitDegrees=0}
]
